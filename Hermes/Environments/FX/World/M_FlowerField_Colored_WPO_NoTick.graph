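# Masked, two-sided material for instanced flower-field debris that drifts on the wind, animated entirely by world position offset.
# Time plus per-instance random loops each instance's age, driving velocity travel with curl noise and axis rotation, suppressed near the camera.
# The wind gust system's fade modulates a random per-instance spawn threshold.
# Dithered opacity uses camera-distance masking, and a flower-field color projection function supplies color.
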Material /Hermes/Environments/FX/World/M_FlowerField_Colored_WPO_NoTick
graph {
 properties {
  bAutomaticallySetUsageInEditor=false
  bUsedWithInstancedStaticMeshes=true
  blendMode=BLEND_Masked
  twoSided=true
  usedShadingModels=Default Lit
 }
 nodes [
  Add(B=LinearInterpolate(A=WorldPosition().XYZ, Alpha=NamedRerouteUsage()))
  n1 = LinearInterpolate(
   A=VectorParameter() {parameterName=Velocity Low, defaultValue=<rgba 1500 0 300 0>}.RGB
   B=VectorParameter() {parameterName=Vel High, defaultValue=<rgba 2250 0 800 0>}.RGB
   Alpha=PerInstanceRandom()
  )
  Add(
   A=@n1
   B=LinearInterpolate(
    B=ScalarParameter() {defaultValue=500, parameterName=Wind Gust Add}
    Alpha=MaterialFunctionCall!(FadeOutTIme, ResetFadeInTime, FadeInTime) {materialFunction=/Game/Environments/TechArt/WindGustSystem/MF_WindGustsSystem}.FadeInFadeOutNormalized
   )
  )
  CollectionParameter() {collection=&FortniteMaterialParameters, parameterName=WindGustsSystem_Active}
  CollectionParameter() {
   collection=&FortniteMaterialParameters
   parameterName=WindGustsSystem_Duration
  }
  CollectionParameter() {
   collection=&FortniteMaterialParameters
   parameterName=WindGustsSystem_TimeStamp
  }
  LinearInterpolate(B=ScalarParameter() {defaultValue=0.06, parameterName=Time High}, Alpha=PerInstanceRandom())
  n2 = NamedRerouteUsage()
  n3 = LinearInterpolate(Alpha=NamedRerouteUsage()) {constA=1, constB=0}
  n4 = SphereMask(
   A=WorldPosition().XYZ
   B=CameraPositionWS()
   Radius=ScalarParameter() {defaultValue=2500, parameterName=Radius2}
   Hardness=ScalarParameter() {defaultValue=500, parameterName=Soften Radius2}
  )
  MaterialFunctionCall(Alpha Threshold=Multiply(A=Multiply(A=@n2, B=@n3), B=OneMinus(None=@n4))) {materialFunction=&DitherTemporalAA}
  n5 = VectorParameter() {parameterName=Rotation Axis, defaultValue=<rgba 1 0 0 0>}
  n6 = Multiply(
   A=Time()
   B=LinearInterpolate(
    A=ScalarParameter() {defaultValue=0.25, parameterName=Rotation Speed Low}
    B=ScalarParameter() {defaultValue=1, parameterName=Rotation SPeed High}
    Alpha=PerInstanceRandom()
   )
  )
  MaterialFunctionCall(Rotation Axis=@n5.RGB, Rotation Angle=@n6) {materialFunction=&FixRotateAboutAxisNormals}
  n7 = Time()
  n8 = Add(A=WorldPosition().XYZ, B=Multiply(A=VectorParameter() {parameterName=Param_1, defaultValue=<rgba -125 0 -25 0>}.RGB, B=@n7))
  n9 = ScalarParameter() {defaultValue=0.002}
  n10 = Multiply(
   A=VectorNoise(World Position=Multiply(A=@n8, B=@n9)) {noiseFunction=VNF_CurlALU}
   B=ScalarParameter() {defaultValue=15, parameterName=Param_1}
  )
  n11 = ScalarParameter() {defaultValue=0.125, parameterName=Time Scale}
  n12 = Fmod(A=Add(A=PerInstanceRandom(), B=Multiply(A=@n7, B=@n11)), B=ScalarParameter() {defaultValue=1, parameterName=Lifetime})
  n13 = Multiply(A=@n12, B=@n1)
  n14 = MaterialFunctionCall() {materialFunction=&ObjectPivotPoint}
  n15 = WorldPosition()
  n16 = WorldPosition()
  n17 = CameraPositionWS()
  n18 = ScalarParameter() {defaultValue=1500, parameterName=Radius}
  n19 = ScalarParameter() {parameterName=Soften Radius}
  Multiply(
   A=Add(
    A=Add(A=@n10, B=@n13)
    B=RotateAboutAxis(NormalizedRotationAxis=@n5.RGB, RotationAngle=@n6, PivotPoint=@n14.Object Pivot Location, Position=@n15.XYZ)
   )
   B=OneMinus(None=SphereMask(A=@n16.XYZ, B=@n17, Radius=@n18, Hardness=@n19))
  )
  Multiply()
  Multiply(
   A=MaterialFunctionCall() {materialFunction=/Game/Global/Base4Layers_2022/MaterialFunctions/MF_FlowerFieldsColorProjection}.Result
   B=ScalarParameter() {defaultValue=0.5, parameterName=Param_3}
  )
  Multiply()
  NamedRerouteDeclaration(None=@n12) {name=Normalized age, nodeColor=<rgba 0 0.7891 1 1>}
  n20 = ScalarParameter() {defaultValue=0.75, parameterName=Param_5}
  n21 = ScalarParameter() {defaultValue=1, parameterName=Param_2}
  n22 = MaterialFunctionCall() {materialFunction=&MF_WindGustsSystem}
  NamedRerouteDeclaration(None=Step(Y=PerInstanceRandom(), X=LinearInterpolate(A=@n20, B=@n21, Alpha=@n22.FadeInFadeOutNormalized))) {name=Spawn, nodeColor=<rgba 0.4062 0 1 1>}
  ScalarParameter() {defaultValue=1, parameterName=Param_4}
  TransformPosition() {transformSourceType=TRANSFORMPOSSOURCE_World, transformType=TRANSFORMPOSSOURCE_Instance}
  WorldPosition()
 ]
}
